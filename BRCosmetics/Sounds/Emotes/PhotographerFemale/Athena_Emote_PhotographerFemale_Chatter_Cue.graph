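# Sound cue for the female photographer emote's vocal chatter, playing one of eight voice variations.
# An equally weighted random node, set to preselect five at level load, feeds a short random delay, then a modulator varying volume and slightly raised pitch.
# It uses emote attenuation, the music emote sound class and emote foley concurrency, and declares an audio-analysis music submix send.
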SoundCue /BRCosmetics/Sounds/Emotes/PhotographerFemale/Athena_Emote_PhotographerFemale_Chatter_Cue
root=Modulator_0
nodes [
 {id=Delay_2, children [Random_0], properties {delayMax=0.5, delayMin=0.25}}
 {id=Modulator_0, children [Delay_2], properties {pitchMax=1.1, pitchMin=1, volumeMax=0.65, volumeMin=0.5}}
 {
  id=Random_0
  children [WavePlayer_2,WavePlayer_3,WavePlayer_4,WavePlayer_5,WavePlayer_6,WavePlayer_7,WavePlayer_8,WavePlayer_9]
  properties {preselectAtLevelLoad=5, weights [1,1,1,1,1,1,1,1]}
 }
 {id=WavePlayer_2, properties {soundWaveAssetPtr=&/Emote_Vocal_PhotographerFemale_Chatter_02}}
 {id=WavePlayer_3, properties {soundWaveAssetPtr=&/Emote_Vocal_PhotographerFemale_Chatter_01}}
 {id=WavePlayer_4, properties {soundWaveAssetPtr=&/Emote_Vocal_PhotographerFemale_Chatter_03}}
 {id=WavePlayer_5, properties {soundWaveAssetPtr=&/Emote_Vocal_PhotographerFemale_Chatter_04}}
 {id=WavePlayer_6, properties {soundWaveAssetPtr=&/Emote_Vocal_PhotographerFemale_Chatter_05}}
 {id=WavePlayer_7, properties {soundWaveAssetPtr=&/Emote_Vocal_PhotographerFemale_Chatter_06}}
 {id=WavePlayer_8, properties {soundWaveAssetPtr=&/Emote_Vocal_PhotographerFemale_Chatter_07}}
 {id=WavePlayer_9, properties {soundWaveAssetPtr=&/Emote_Vocal_PhotographerFemale_Chatter_08}}
]
properties {
 attenuationSettings=&Fort_Emotes_Athena_Att
 maxDistance=1600
 soundClassObject=&Fort_Music_Emote
 volumeMultiplier=1
 concurrencyOverrides {maxCount=3, resolutionRule=StopOldest}
 concurrencySet [/Game/Sounds/Concurrency_Rules/Athena_EmoteFoley_Conc]
 soundSubmixSends [{minSendLevel=0, soundSubmix=&Fortnite_AudioAnalysis_Music}]
}
