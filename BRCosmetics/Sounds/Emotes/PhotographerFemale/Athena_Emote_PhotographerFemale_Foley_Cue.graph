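# Sound cue for the female Photographer emote's foley, mixing a start sound, a looping foley bed and a vocal chatter line.
# The start wave plays immediately, while fixed delays bring in the loop and, earlier than it, one of three randomly chosen chatter clips with slight volume modulation.
# It uses emote foley class, attenuation and concurrency settings, and sends to a music audio-analysis submix.
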SoundCue /BRCosmetics/Sounds/Emotes/PhotographerFemale/Athena_Emote_PhotographerFemale_Foley_Cue
root=Mixer_0
nodes [
 {id=Delay_0, children [WavePlayer_1], properties {delayMax=1.6, delayMin=1.6}}
 {id=Delay_1, children [Random_1], properties {delayMax=0.9, delayMin=0.9}}
 {id=Mixer_0, children [WavePlayer_0,Delay_0,Modulator_0], properties {inputVolume [1,0.85,1]}}
 {id=Modulator_0, children [Delay_1], properties {volumeMax=1.3, volumeMin=1}}
 {id=Random_1, children [WavePlayer_10, WavePlayer_11, WavePlayer_12], properties {weights [1,1,1]}}
 {id=WavePlayer_0, properties {soundWaveAssetPtr=&/Emote_Foley_PhotographerFemale_Start}}
 {
  id=WavePlayer_1
  properties {bLooping=true, soundWaveAssetPtr=&/Emote_Foley_PhotographerFemale_Loop}
 }
 {id=WavePlayer_10, properties {soundWaveAssetPtr=&/Emote_Vocal_PhotographerFemale_Chatter_06}}
 {id=WavePlayer_11, properties {soundWaveAssetPtr=&/Emote_Vocal_PhotographerFemale_Chatter_04}}
 {id=WavePlayer_12, properties {soundWaveAssetPtr=&/Emote_Vocal_PhotographerFemale_Chatter_05}}
]
properties {
 attenuationSettings=&Fort_Emotes_Athena_Att
 audioPropertiesSheet=/BRCosmetics/Sounds/Settings/PropertySheets/Emotes/APS_Cosmetic_Emote_Foley
 maxDistance=1600
 soundClassObject=/Game/Sounds/FortSoundClassesAndModes/Fort_Foley_Emote
 volumeMultiplier=1
 concurrencyOverrides {maxCount=3, resolutionRule=StopOldest}
 concurrencySet [/Game/Sounds/Concurrency_Rules/Athena_EmoteFoley_Conc]
 soundSubmixSends [{minSendLevel=0, soundSubmix=&Fortnite_AudioAnalysis_Music}]
}
